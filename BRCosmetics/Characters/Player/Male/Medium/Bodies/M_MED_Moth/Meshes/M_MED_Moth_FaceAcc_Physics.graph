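# Physics asset for the face accessory mesh of the male medium Moth player cosmetic, defining collision and joints for its antennae and proboscis.
# A head sphere anchors two three-segment capsule antenna chains and a two-segment proboscis chain.
# Constraints chain each segment to its parent with limited swing and near-locked twist.
# Outer antenna joints restrict the second swing axis tightly, so they bend mostly along one axis.
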PhysicsAsset /BRCosmetics/Characters/Player/Male/Medium/Bodies/M_MED_Moth/Meshes/M_MED_Moth_FaceAcc_Physics
bodies {
 dyn_antennae01_l [{shapeType=Capsule, length=8.0154, radius=3.5177, center=<7.476 -0.1971 -0.3834>, rotation=<rot 0 -95.4431 -84.3498>}]
 dyn_antennae01_r [{shapeType=Capsule, length=8.177, radius=3.5141, center=<7.6694 -0.2495 1.3784>, rotation=<rot 0 84.5835 -82.7582>}]
 dyn_antennae02_l [{shapeType=Capsule, length=4.5476, radius=3.5583, center=<7.7837 -2.5781 -0.678>, rotation=<rot 0 -107.7232 -82.7698>}]
 dyn_antennae02_r [{shapeType=Capsule, length=3.7256, radius=3.2936, center=<7.9677 -0.4146 0.0465>, rotation=<rot -0.7067 88.5167 -88.1519>}]
 dyn_antennae03_l [{shapeType=Capsule, length=3.1267, radius=1.9609, center=<4.1618 -0.3472 -0.7071>, rotation=<rot 0 -96.7955 -82.4836>}]
 dyn_antennae03_r [{shapeType=Capsule, length=4.0192, radius=2.0261, center=<4.1841 -0.5143 -0.0881>, rotation=<rot 0 70.2627 -82.9983>}]
 head [{shapeType=Sphere, radius=12.8306, center=<-2.3627 4.4429 0.0187>}]
 proboscis01 [{shapeType=Capsule, length=3.5489, radius=2.1464, center=<-4.6378 0.2763 0.0933>, rotation=<rot 0 -86.7295 -88.763>}]
 proboscis02 [{shapeType=Capsule, length=5.131, radius=1.9677, center=<-4.1677 -0.5659 0.2867>, rotation=<rot 0 -80.707 -88.7815>}]
}
constraints [
 {bone1Name=proboscis02, bone2Name=proboscis01, swing1LimitDegrees=20, swing2LimitDegrees=20, twistLimitDegrees=1}
 {bone1Name=dyn_antennae02_l, bone2Name=dyn_antennae01_l, swing1LimitDegrees=20, swing2LimitDegrees=3, twistLimitDegrees=1}
 {bone1Name=dyn_antennae03_l, bone2Name=dyn_antennae02_l, swing1LimitDegrees=20, swing2LimitDegrees=3, twistLimitDegrees=1}
 {bone1Name=dyn_antennae02_r, bone2Name=dyn_antennae01_r, swing1LimitDegrees=20, swing2LimitDegrees=3, twistLimitDegrees=1}
 {bone1Name=dyn_antennae03_r, bone2Name=dyn_antennae02_r, swing1LimitDegrees=20, swing2LimitDegrees=3, twistLimitDegrees=1}
 {bone1Name=dyn_antennae01_l, bone2Name=head, swing1LimitDegrees=20, swing2LimitDegrees=20, twistLimitDegrees=1}
 {bone1Name=dyn_antennae01_r, bone2Name=head, swing1LimitDegrees=20, swing2LimitDegrees=20, twistLimitDegrees=1}
 {bone1Name=proboscis01, bone2Name=head, swing1LimitDegrees=20, swing2LimitDegrees=20, twistLimitDegrees=1}
]
