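# Physics asset for the fluff mesh on the medium male Moth player cosmetic, giving its fluff tufts constrained secondary motion.
# Seven tufts ring the back and sides, each pairing a tiny base capsule locked to spine_05 with a larger dyn capsule allowed slight swing.
# A head sphere, upper-spine box and clavicle and upper-arm capsules form body shapes.
# Head and clavicles rotate freely against the spine, while upper arms lock to clavicles.
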PhysicsAsset /BRCosmetics/Characters/Player/Male/Medium/Bodies/M_MED_Moth/Meshes/M_MED_MothFluff_Physics
bodies {
 base_fluff_back [{shapeType=Capsule, length=0.505, radius=0.505, shapeName=base_fluff_back_capsule, rotation=<rot 0 -90 0>}]
 base_fluff_left [{shapeType=Capsule, length=0.505, radius=0.505, shapeName=base_fluff_left_capsule, rotation=<rot 0 -90 0>}]
 base_fluff_leftback [{shapeType=Capsule, length=0.505, radius=0.505, shapeName=base_fluff_leftback_capsule, rotation=<rot 0 -90 0>}]
 base_fluff_leftfront [{shapeType=Capsule, length=0.505, radius=0.505, shapeName=base_fluff_leftfront_capsule, rotation=<rot 0 -90 0>}]
 base_fluff_right [{shapeType=Capsule, length=0.505, radius=0.505, shapeName=base_fluff_right_capsule, rotation=<rot 0 -90 0>}]
 base_fluff_rightback [{shapeType=Capsule, length=0.505, radius=0.505, shapeName=base_fluff_rightback_capsule, rotation=<rot 0 -90 0>}]
 base_fluff_rightfront [{shapeType=Capsule, length=0.505, radius=0.505, shapeName=base_fluff_rightfront_capsule, rotation=<rot 0 -90 0>}]
 clavicle_l [{shapeType=Capsule, length=7, radius=5, shapeName=clavicle_l_capsule, center=<5.3485 4 0>, rotation=<rot 90 0 0>}]
 clavicle_r [{shapeType=Capsule, length=7, radius=5, shapeName=clavicle_r_capsule, center=<-5.3485 -4 0>, rotation=<rot 90 180 0>}]
 dyn_fluff_back [{shapeType=Capsule, length=6.5, radius=4.5, shapeName=dyn_fluff_back_capsule, center=<2 -2 0>}]
 dyn_fluff_left [{shapeType=Capsule, length=7, radius=4, shapeName=dyn_fluff_left_capsule, center=<2 -2 0>}]
 dyn_fluff_leftback [{shapeType=Capsule, length=6, radius=5, shapeName=dyn_fluff_leftback_capsule, center=<2 -2 0>, rotation=<rot 0.0001 -0.0001 0>}]
 dyn_fluff_leftfront [{shapeType=Capsule, length=6, radius=5, shapeName=dyn_fluff_leftfront_capsule, center=<2 -2 0>}]
 dyn_fluff_right [{shapeType=Capsule, length=7, radius=4, shapeName=dyn_fluff_right_capsule, center=<2 -2 0>}]
 dyn_fluff_rightback [{shapeType=Capsule, length=6, radius=5, shapeName=dyn_fluff_rightback_capsule, center=<2 -2 0>}]
 dyn_fluff_rightfront [{shapeType=Capsule, length=7, radius=4, shapeName=dyn_fluff_rightfront_capsule, center=<2 -2 0>}]
 head [{shapeType=Sphere, radius=10.505, center=<4.0044 0.6599 0>}]
 spine_05 [{shapeType=Box, extentX=16.926, extentY=19.2177, extentZ=28.3345, center=<-2.2642 -0.7982 -0.188>, rotation=<rot 0 -171.5747 -89.5735>}]
 upperarm_l [{shapeType=Capsule, length=13, radius=7, shapeName=upperarm_l_capsule, center=<13 0 1.5>, rotation=<rot 90 0 0>}]
 upperarm_r [{shapeType=Capsule, length=13, radius=7, shapeName=upperarm_r_capsule, center=<-13 0 -1.5>, rotation=<rot 90 0 0>}]
}
constraints [
 {
  bone1Name=head
  bone2Name=spine_05
  swing1LimitDegrees=45
  swing1Motion=Free
  swing2LimitDegrees=45
  swing2Motion=Free
  twistLimitDegrees=45
  twistMotion=Free
 }
 {
  bone1Name=base_fluff_back
  bone2Name=spine_05
  swing1LimitDegrees=45
  swing1Motion=Locked
  swing2LimitDegrees=45
  swing2Motion=Locked
  twistLimitDegrees=45
  twistMotion=Locked
 }
 {bone1Name=dyn_fluff_back, bone2Name=base_fluff_back, swing1LimitDegrees=5, swing2LimitDegrees=5, twistLimitDegrees=0}
 {
  bone1Name=base_fluff_left
  bone2Name=spine_05
  swing1LimitDegrees=45
  swing1Motion=Locked
  swing2LimitDegrees=45
  swing2Motion=Locked
  twistLimitDegrees=45
  twistMotion=Locked
 }
 {bone1Name=dyn_fluff_left, bone2Name=base_fluff_left, swing1LimitDegrees=5, swing2LimitDegrees=5, twistLimitDegrees=0}
 {
  bone1Name=base_fluff_leftback
  bone2Name=spine_05
  swing1LimitDegrees=45
  swing1Motion=Locked
  swing2LimitDegrees=45
  swing2Motion=Locked
  twistLimitDegrees=45
  twistMotion=Locked
 }
 {bone1Name=dyn_fluff_leftback, bone2Name=base_fluff_leftback, swing1LimitDegrees=5, swing2LimitDegrees=5, twistLimitDegrees=0}
 {
  bone1Name=base_fluff_leftfront
  bone2Name=spine_05
  swing1LimitDegrees=45
  swing1Motion=Locked
  swing2LimitDegrees=45
  swing2Motion=Locked
  twistLimitDegrees=45
  twistMotion=Locked
 }
 {bone1Name=dyn_fluff_leftfront, bone2Name=base_fluff_leftfront, swing1LimitDegrees=5, swing2LimitDegrees=5, twistLimitDegrees=0}
 {
  bone1Name=base_fluff_right
  bone2Name=spine_05
  swing1LimitDegrees=45
  swing1Motion=Locked
  swing2LimitDegrees=45
  swing2Motion=Locked
  twistLimitDegrees=45
  twistMotion=Locked
 }
 {bone1Name=dyn_fluff_right, bone2Name=base_fluff_right, swing1LimitDegrees=5, swing2LimitDegrees=5, twistLimitDegrees=0}
 {
  bone1Name=base_fluff_rightback
  bone2Name=spine_05
  swing1LimitDegrees=45
  swing1Motion=Locked
  swing2LimitDegrees=45
  swing2Motion=Locked
  twistLimitDegrees=45
  twistMotion=Locked
 }
 {bone1Name=dyn_fluff_rightback, bone2Name=base_fluff_rightback, swing1LimitDegrees=5, swing2LimitDegrees=5, twistLimitDegrees=0}
 {
  bone1Name=base_fluff_rightfront
  bone2Name=spine_05
  swing1LimitDegrees=45
  swing1Motion=Locked
  swing2LimitDegrees=45
  swing2Motion=Locked
  twistLimitDegrees=45
  twistMotion=Locked
 }
 {bone1Name=dyn_fluff_rightfront, bone2Name=base_fluff_rightfront, swing1LimitDegrees=5, swing2LimitDegrees=5, twistLimitDegrees=0}
 {
  bone1Name=upperarm_l
  bone2Name=clavicle_l
  swing1LimitDegrees=45
  swing1Motion=Locked
  swing2LimitDegrees=45
  swing2Motion=Locked
  twistLimitDegrees=45
  twistMotion=Locked
 }
 {
  bone1Name=upperarm_r
  bone2Name=clavicle_r
  swing1LimitDegrees=45
  swing1Motion=Locked
  swing2LimitDegrees=45
  swing2Motion=Locked
  twistLimitDegrees=45
  twistMotion=Locked
 }
 {
  bone1Name=clavicle_l
  bone2Name=spine_05
  swing1LimitDegrees=45
  swing1Motion=Free
  swing2LimitDegrees=45
  swing2Motion=Free
  twistLimitDegrees=45
  twistMotion=Free
 }
 {
  bone1Name=clavicle_r
  bone2Name=spine_05
  swing1LimitDegrees=45
  swing1Motion=Free
  swing2LimitDegrees=45
  swing2Motion=Free
  twistLimitDegrees=45
  twistMotion=Free
 }
]
