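# Physics asset for the M_MED_Moth male medium player body cosmetic, covering its waist straps and jacket flap.
# Simple box and sphere bodies on pelvis, spine_05 and head form a core chain linked by free angular constraints.
# The jacket flap and upper straps attach to the pelvis with tight angular limits.
# Tiny base strap capsules lock to the pelvis as anchors for the narrowly limited strap_01 capsules.
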PhysicsAsset /BRCosmetics/Characters/Player/Male/Medium/Bodies/M_MED_Moth/Meshes/M_MED_Moth_Physics
bodies {
 base_waist_strap_l_01 [{shapeType=Capsule, length=0.505, radius=0.505, shapeName=base_waist_strap_l_01_capsule, rotation=<rot 0 -90 0>}]
 base_waist_strap_r_01 [{shapeType=Capsule, length=0.505, radius=0.505, shapeName=base_waist_strap_r_01_capsule, rotation=<rot 0 -90 0>}]
 dyn_jacket_flap [{shapeType=Capsule, length=5.1179, radius=3.8554, shapeName=dyn_jacket_flap_capsule, center=<-5 -2.2 0>}]
 dyn_waist_strap_l [{shapeType=Box, extentX=11, extentY=11, extentZ=11, rotation=<rot 0 -90 0>}]
 dyn_waist_strap_l_01 [{shapeType=Capsule, length=1.9379, radius=1.2284, shapeName=dyn_waist_strap_l_01_capsule, center=<2 0 0>, rotation=<rot 90 0 0>}]
 dyn_waist_strap_r [{shapeType=Box, extentX=10, extentY=10, extentZ=10, rotation=<rot 0 -90 0>}]
 dyn_waist_strap_r_01 [{shapeType=Capsule, length=2.2546, radius=1.4536, shapeName=dyn_waist_strap_r_01_capsule, center=<-2 0 0>, rotation=<rot 90 0 0>}]
 head [{shapeType=Sphere, radius=11.005, center=<4.0044 0.6599 0>}]
 pelvis [{shapeType=Box, extentX=20.574, extentY=21.9919, extentZ=34.3604, center=<-3.1385 0.281 -0.4303>, rotation=<rot -6.2235 -94.328 -1.2334>}]
 spine_05 [{shapeType=Box, extentX=16.926, extentY=19.2177, extentZ=28.3345, center=<-2.2642 -0.7982 -0.188>, rotation=<rot 0 -171.5747 -89.5735>}]
}
constraints [
 {
  bone1Name=head
  bone2Name=spine_05
  swing1LimitDegrees=45
  swing1Motion=Free
  swing2LimitDegrees=45
  swing2Motion=Free
  twistLimitDegrees=45
  twistMotion=Free
 }
 {
  bone1Name=spine_05
  bone2Name=pelvis
  swing1LimitDegrees=45
  swing1Motion=Free
  swing2LimitDegrees=45
  swing2Motion=Free
  twistLimitDegrees=45
  twistMotion=Free
 }
 {bone1Name=dyn_waist_strap_l, bone2Name=pelvis, swing1LimitDegrees=20, swing2LimitDegrees=1, twistLimitDegrees=1}
 {bone1Name=dyn_waist_strap_r, bone2Name=pelvis, swing1LimitDegrees=20, swing2LimitDegrees=1, twistLimitDegrees=1}
 {
  bone1Name=base_waist_strap_l_01
  bone2Name=pelvis
  swing1LimitDegrees=45
  swing1Motion=Locked
  swing2LimitDegrees=45
  swing2Motion=Locked
  twistLimitDegrees=45
  twistMotion=Locked
 }
 {bone1Name=dyn_waist_strap_l_01, bone2Name=base_waist_strap_l_01, swing1LimitDegrees=10, swing2LimitDegrees=5, twistLimitDegrees=0}
 {
  bone1Name=base_waist_strap_r_01
  bone2Name=pelvis
  swing1LimitDegrees=45
  swing1Motion=Locked
  swing2LimitDegrees=45
  swing2Motion=Locked
  twistLimitDegrees=45
  twistMotion=Locked
 }
 {bone1Name=dyn_waist_strap_r_01, bone2Name=base_waist_strap_r_01, swing1LimitDegrees=10, swing2LimitDegrees=5, twistLimitDegrees=0}
 {bone1Name=dyn_jacket_flap, bone2Name=pelvis, swing1LimitDegrees=30, swing2LimitDegrees=0, twistLimitDegrees=0}
]
