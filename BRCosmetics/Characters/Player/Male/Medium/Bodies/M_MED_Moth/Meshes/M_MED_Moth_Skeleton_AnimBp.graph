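# Post-copy animation blueprint for the Moth male medium body cosmetic, adding procedural and physics motion to the pose copied from the attached parent mesh.
# A Control Rig feeds two chained RigidBody nodes: fluff physics relative to spine_05, then body physics relative to pelvis.
# Each update zeroes the rig's NotGliding alpha during skydiving and stores a computed Gravity Override, though both physics nodes show fixed downward gravity.
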AnimBlueprint /BRCosmetics/Characters/Player/Male/Medium/Bodies/M_MED_Moth/Meshes/M_MED_Moth_Skeleton_AnimBp
target_skeleton=&/M_MED_Moth_Skeleton
graphs [
 AnimGraph {nodes [
   n1 = Control Rig!(
    Source=Copy Pose From Mesh() {bUseAttachedParent=true}
    Alpha
    NotGliding_Alpha:double=Get InverseGliding_Alha() -> (InverseGliding_Alha:double)
   ) -> (Pose) {controlRigAssetReference {blueprintRigClass=&/M_MED_Moth_CtrlRig.M_MED_Moth_CtrlRig_C}}
   n2 = RigidBody!(ComponentPose=Local To Component(LocalPose=@n1), Alpha, OverrideWorldGravity:Vector="0,0,-980", ExternalForce) -> (Pose) {
    bOverrideWorldGravity=true
    overridePhysicsAsset=&/M_MED_MothFluff_Physics
    simulationSpace=BaseBoneSpace
    alphaBoolBlend {blendOption=Linear}
    baseBoneRef {boneName=spine_05}
    overrideWorldGravity {z=-980}
    simSpaceSettings {
     maxAngularAcceleration=40
     maxAngularVelocity=8
     maxLinearAcceleration=3500
     maxLinearVelocity=700
     worldAlpha=1
     externalLinearDragV=<0.3 0.3 0.3>
    }
   }
   n3 = RigidBody(ComponentPose=@n2, OverrideWorldGravity="0,0,-980") {
    bOverrideWorldGravity=true
    overridePhysicsAsset=&/M_MED_Moth_Physics
    simulationSpace=BaseBoneSpace
    alphaBoolBlend {blendOption=Linear}
    baseBoneRef {boneName=pelvis}
    overrideWorldGravity {z=-980}
    simSpaceSettings {
     maxAngularAcceleration=40
     maxAngularVelocity=8
     maxLinearAcceleration=3500
     maxLinearVelocity=700
     worldAlpha=1
     externalLinearDragV=<0.3 0.3 0.3>
    }
   }
   Output Pose(Result=Component To Local(ComponentPose=@n3)) {layerGroup=None}
  ]}
 EventGraph {nodes [
   Event BlueprintUpdateAnimation()
   > n1 = "Gravity Override State Function (Parameter Asset) (Nativized)"(
    CharacterPartAnimInstance=Self-Reference()
    GameTimeInSeconds=GetGameTimeInSeconds()
    Parameters=&/M_MED_Moth_Parameters
   )
   > Set Gravity Override(Gravity Override=@n1)
   > n2 = Branch(Condition=Get bIsSkyDiving())
   > Set InverseGliding_Alha(InverseGliding_Alha:double=0) -> (Output_Get:double)
   Set InverseGliding_Alha(execute=@n2.else, InverseGliding_Alha=1)
   TryGetPawnOwner()
   Event OnBeginSkydiving()
   Event OnEndSkydiving()
  ]}
]
variables [
 Gravity Override {type=Vector}
 InverseGliding_Alha {type=number}
]
